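# Material function for the Akimbo Envoy pickaxe that replaces emissive with an animated reactivity glow, blended from black by Reactivity Alpha.
# Waves use time-driven sine pulses on UV1, quantized by Ceil into steps and shaped by Frac bands.
# The core pulses through a sine-driven sphere mask radius, less a small subtracted term.
# Named reroutes supply tech-art texture masks, the emissive texture and input roughness to both strength-scaled layers, which are summed.
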MaterialFunction /BRCosmetics/Weapons/FORT_Melee/Pickaxe_AkimboEnvoy/Materials/MF_Pickaxe_AkimboEnvoy_Reactivity
graph {nodes [
  n1 = GetMaterialAttributes(None=FunctionInput() {inputName=Material, inputType=FunctionInput_MaterialAttributes}) {attributeGetTypes [#1, #2]}
  n2 = Constant()
  n3 = MaterialFunctionCall(Speed=ScalarParameter() {defaultValue=1, group=Reactivity, parameterName=Reactivity Speed, sortPriority=2}) {materialFunction=&TimeWithSpeedVariable}
  n4 = ScalarParameter() {defaultValue=4, group=Reactivity, parameterName=Wave Steps}
  n5 = Multiply(A=TextureCoordinate() {coordinateIndex=1}, B=@n4) {constB=4}
  n6 = Add(A=Divide(A=Ceil(None=@n5), B=@n4) {constB=4}, B=ScalarParameter() {group=Reactivity, parameterName=Wave Bias})
  n7 = Multiply(
   A=MaterialFunctionCall(UV Channel=Subtract(A=@n3.Time, B=@n6)) {materialFunction=&LinearGradient}.VGradient
   B=ScalarParameter() {defaultValue=1, group=Reactivity, parameterName=Wave Frequency}
  )
  n8 = Power(Base=Saturate(None=Sine(None=@n7))) {constExponent=5}
  n9 = Multiply(A=TextureCoordinate() {coordinateIndex=1}, B=ScalarParameter() {defaultValue=4, group=Reactivity, parameterName=Wave Steps}) {constB=4}
  n10 = Subtract(A=Multiply(A=Frac(None=@n9)) {constB=2})
  n11 = OneMinus(None=ComponentMask(None=Abs(None=@n10)) {g=true})
  n12 = Multiply(A=NamedRerouteUsage(), B=Saturate(None=Multiply(A=@n8, B=@n11)))
  n13 = ScalarParameter() {defaultValue=10, group=Reactivity, parameterName=Wave Strength}
  n14 = NamedRerouteUsage()
  n15 = VectorParameter() {group=Reactivity, parameterName=Core Position, defaultValue=<rgba 0.25 0.75 0 1>}
  n16 = ScalarParameter() {defaultValue=1, group=Reactivity, parameterName=Core Max Radius}
  n17 = ScalarParameter() {defaultValue=1, group=Reactivity, parameterName=Reactivity Speed, sortPriority=2}
  n18 = Add(
   A=Multiply(
    A=MaterialFunctionCall(Speed=@n17) {materialFunction=&TimeWithSpeedVariable}.Time
    B=ScalarParameter() {defaultValue=1, group=Reactivity, parameterName=Core Multiplier}
   )
   B=ScalarParameter() {group=Reactivity, parameterName=Core Bias}
  )
  n19 = Power(Base=Saturate(None=Sine(None=@n18))) {constExponent=7}
  n20 = LinearInterpolate(
   A=@n19
   B=SphereMask(A=TextureCoordinate() {coordinateIndex=1}, B=AppendVector(A=@n15.R, B=@n15.R), Radius=LinearInterpolate(B=@n16, Alpha=@n19)) {attenuationRadius=0.1, hardnessPercent=1}
  ) {constAlpha=0.2}
  n21 = NamedRerouteUsage()
  n22 = Saturate(None=Subtract(A=Multiply(A=@n14, B=@n20), B=LinearInterpolate(Alpha=@n21) {constB=0.15}))
  n23 = ScalarParameter() {defaultValue=50, group=Reactivity, parameterName=Core Strength}
  n24 = Add(A=Multiply(A=Multiply(A=@n12, B=@n13), B=NamedRerouteUsage()), B=Multiply(A=Multiply(A=@n22, B=@n23), B=NamedRerouteUsage()))
  n25 = ScalarParameter() {group=Reactivity, parameterName=Reactivity Alpha, sortPriority=1}
  FunctionOutput(None=SetMaterialAttributes!(MaterialAttributes=@n1.MaterialAttributes, EmissiveColor=LinearInterpolate(A=@n2, B=@n24, Alpha=@n25) {constAlpha=0}) {attributeSetTypes [#1]})
  n26 = TextureSample() {samplerType=SAMPLERTYPE_Masks, texture=/BRCosmetics/Weapons/FORT_Melee/Pickaxe_AkimboEnvoy/Textures/T_Pickaxe_AkimboEnvoy_TechArt}
  NamedRerouteDeclaration(None=@n26.R) {name=Wave Mask, nodeColor=<rgba 1 0.5859 0 1>}
  NamedRerouteDeclaration(None=@n26.G) {name=Core Mask, nodeColor=<rgba 0.9453 1 0 1>}
  NamedRerouteDeclaration(None=TextureSample() {texture=/BRCosmetics/Weapons/FORT_Melee/Pickaxe_AkimboEnvoy/Textures/T_Pickaxe_AkimboEnvoy_E}.RGB) {name=Input Emissive, nodeColor=<rgba 1 0.9375 0 1>}
  NamedRerouteDeclaration(None=@n1.Roughness) {name=Input Roughness, nodeColor=<rgba 1 0.0469 0 1>}
 ]}
